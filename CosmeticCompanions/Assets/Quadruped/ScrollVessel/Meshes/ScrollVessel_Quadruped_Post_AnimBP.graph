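# Post-process Animation Blueprint for the ScrollVessel quadruped cosmetic companion's skeleton.
# The incoming pose runs through a Pose Blend node using ScrollVessel_PoseAsset, which applies curve-driven poses before reaching the output.
# A second, linear Pose Blend node reads the template-defaulted PoseAsset variable but is not connected to the output.
# The event graph is empty.
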AnimBlueprint /CosmeticCompanions/Assets/Quadruped/ScrollVessel/Meshes/ScrollVessel_Quadruped_Post_AnimBP
target_skeleton=&/ScrollVessel_Quadruped_Skeleton
graphs [
 AnimGraph {nodes [
   Pose!(PoseAsset:Pose Asset Object Reference=Get PoseAsset() -> (PoseAsset:Pose Asset Object Reference), SourcePose:Pose Link Structure) -> (Pose:Pose Link Structure) {is=title AnimGraphNode_PoseBlendNode, blendOption=Linear}
   Output Pose(Result=ScrollVessel_PoseAsset!(SourcePose:Pose Link Structure=Input Pose()) -> (Pose:Pose Link Structure) {is=title AnimGraphNode_PoseBlendNode, poseAsset=&/ScrollVessel_PoseAsset})
  ]}
 EventGraph {links [], nodes []}
]
variables [PoseAsset {type=PoseAsset, default=/CosmeticCompanions/Automation_Template/companion_pose_asset_template}]
